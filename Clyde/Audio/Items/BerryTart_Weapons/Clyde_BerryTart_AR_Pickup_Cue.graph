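# Sound cue for picking up the Clyde BerryTart assault rifle, using loot-collect sound class and weapon pickup attenuation.
# A quality-level node routes to two licensed-content switchers with equivalent wave content.
# Each switcher offers the plain rifle pickup on two inputs and, on the third, a mix with the ClydeShark music stinger.
# Pickup concurrency sets limit overlapping plays, and a pre-effect send feeds the boss VO envelope-follower bus.
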SoundCue /Clyde/Audio/Items/BerryTart_Weapons/Clyde_BerryTart_AR_Pickup_Cue
root=QualityLevel_0
nodes [
 {id=FortSoundNodeLicensedContentSwitcher_0, class=FortSoundNodeLicensedContentSwitcher, children [WavePlayer_4,WavePlayer_4,Mixer_1]}
 {id=FortSoundNodeLicensedContentSwitcher_1, class=FortSoundNodeLicensedContentSwitcher, children [WavePlayer_5,WavePlayer_5,Mixer_0]}
 {id=Mixer_0, children [WavePlayer_5,WavePlayer_3], properties {inputVolume [1.2,1]}}
 {id=Mixer_1, children [WavePlayer_4,WavePlayer_2], properties {inputVolume [1.2,1]}}
 {
  id=QualityLevel_0
  children [
   FortSoundNodeLicensedContentSwitcher_0
   FortSoundNodeLicensedContentSwitcher_0
   FortSoundNodeLicensedContentSwitcher_1
   FortSoundNodeLicensedContentSwitcher_1
  ]
 }
 {id=WavePlayer_2, properties {soundWaveAssetPtr=&/MX_IP_ClydeShark_AR_Pickup_Stinger}}
 {id=WavePlayer_3, properties {soundWaveAssetPtr=&MX_IP_ClydeShark_AR_Pickup_Stinger}}
 {id=WavePlayer_4, properties {soundWaveAssetPtr=&/Clyde_Rifle_Pickup}}
 {id=WavePlayer_5, properties {soundWaveAssetPtr=&Clyde_Rifle_Pickup}}
]
properties {
 attenuationSettings=&Fort_Weapon_Pickup_Att
 maxDistance=3500
 soundClassObject=&Fort_Loot_Collect
 volumeMultiplier=0.7
 concurrencyOverrides {maxCount=1, resolutionRule=PreventNew}
 concurrencySet [&PickupMaximum_Oldest, /Clyde/Audio/Items/Clyde_Weapon_Pickup_Concurrency]
 preEffectBusSends [{audioBus=/Game/Sounds/Modulation/EnvelopeFollowers/BR_Boss/AB_BR_Boss_VO_EnvFollow, soundSourceBus=null}]
}
