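# Sound cue for picking up the Clyde Icy Grapple weapon, using weapon pickup attenuation and the loot-collect sound class.
# A mixer layers a modulated random pick of two bow bone pickup waves with a licensed-content switcher.
# Only the switcher's third branch is filled, with an Icy Grapple music stinger.
# The cue also sends to a boss VO envelope-follower bus and uses two pickup concurrency sets.
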SoundCue /Clyde/Audio/Items/BerryTart_Weapons/Clyde_Icygrapple_Pickup_Cue
root=Mixer_0
nodes [
 {id=FortSoundNodeLicensedContentSwitcher_0, class=FortSoundNodeLicensedContentSwitcher, children [null,null,WavePlayer_0]}
 {id=Mixer_0, children [Modulator_0, FortSoundNodeLicensedContentSwitcher_0], properties {inputVolume [1,1]}}
 {id=Modulator_0, children [Random_1]}
 {id=Random_1, children [WavePlayer_1,WavePlayer_2], properties {weights [1,1]}}
 {id=WavePlayer_0, properties {soundWaveAssetPtr=&/MX_IP_Clyde_IcyGrapple_Pickup_Stinger}}
 {id=WavePlayer_1, properties {soundWaveAssetPtr=&/Clyde_Bow_Bone_Pickup_02}}
 {id=WavePlayer_2, properties {soundWaveAssetPtr=&/Clyde_Bow_Bone_Pickup_01}}
]
properties {
 attenuationSettings=&Fort_Weapon_Pickup_Att
 maxDistance=3500
 soundClassObject=&Fort_Loot_Collect
 volumeMultiplier=0.7
 busSends [{audioBus=/Game/Sounds/Modulation/EnvelopeFollowers/BR_Boss/AB_BR_Boss_VO_EnvFollow, minSendLevel=0, sendLevel=1, soundSourceBus=null}]
 concurrencySet [&PickupMaximum_Oldest, /Clyde/Audio/Items/Clyde_Weapon_Pickup_Concurrency]
}
